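; Quest gameplay ability that tracks continuous vehicle airtime, apparently for a three-second challenge in Battle Pass Season 19.
; On the vehicle-entered event it checks the vehicle against a car tag and starts a looping timer, which the exit event clears.
; Each tick adds the interval while airborne or resets the total on landing.
; Reaching the required time sends a quest stat event and clears the timer.
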(event Custom|Check%20In%20Air
  (Utilities|Casting|CastToFortAthenaVehicle (Vehicle|GetVehicleInterface (Variables|GAVars|Data|GetOwningPlayerPawn))
    (:then
      (if (Vehicle|VehicleIsinAir (Vehicle|GetVehicleInterface (Variables|GAVars|Data|GetOwningPlayerPawn)))
        ;[1 1 1] Add the time spent in air to the total elapsed time in air (as calculated by the time interval)
        (bind _output_get (Variables|Default|SetElapsedTimeInAir (+ (Variables|Default|GetExecutionTimeInterval) (Variables|Default|GetElapsedTimeInAir))))
        (if (>= _output_get (Variables|Default|GetRequiredTime))
          (Quest|SendComplexCustomStatEvent (Variables|GAVars|Data|GetOwningPlayerQuestManager) 0 (Variables|Default|GetT_Quest) "(GameplayTags=)" 1)
          (Utilities|Time|ClearandInvalidateTimerbyHandle (Variables|Default|GetTimer%20Handle)))
        (else
          (Variables|Default|SetElapsedTimeInAir 0.0))))
    (:CastFailed
      _))
  (Ability|EndAbility))

(event Ability|EventActivateAbilityFromEvent
  (params (EventData :type Gameplay%20Event%20Data%20Structure%20%28by%20ref%29))
  (bind _timer_handle (Variables|Default|GetTimer%20Handle))
  (|CallParentFunction EventData)
  (switch Utilities|FlowControl|Switch|SwitchonGameplayTag (Utilities|Struct|BreakGameplayEventData EventData)
    (:Event.Vehicle.Entered
      ;[1 1 1] Check Vehicle Tags
      (if (GameplayTags|TagContainer|CompareTagAssetInterfacetoOtherTags (Variables|Default|GetT_Car) (Vehicle|GetVehicleAbilitySystemComponent (Vehicle|GetVehicleInterface (Variables|GAVars|Data|GetOwningPlayerPawn))))
        ;[1 1 1] Start Checking Vehicle If Vehicle is in Air
        (bind _returnvalue (Utilities|Time|SetTimerbyEvent (delegate Custom|Check%20In%20Air) (Variables|Default|GetExecutionTimeInterval) true))
        (Variables|Default|SetTimer%20Handle _returnvalue)
        (else
          (Ability|EndAbility))))
    (:Event.Vehicle.Exit
      ;[1 1 1] On Vehicle Exit Clear Timer and End Ability
      (if (Utilities|Time|IsValidTimerHandle _timer_handle)
        (Utilities|Time|ClearandInvalidateTimerbyHandle _timer_handle)
        (else
          _))
      (Ability|EndAbility))))
