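; Quest gameplay ability that tracks cumulative vehicle air time for a quest.
; On vehicle entry it checks the vehicle's tags, then runs a looping timer that adds the interval while the vehicle is airborne.
; On reaching the required time or landing, it sends the truncated time as a quest stat event and resets.
; Vehicle exit discards unsent time; exit or quest completion clears the timer and ends the ability.
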(event Custom|Check%20In%20Air
  (Utilities|Casting|CastToFortAthenaVehicle (Vehicle|GetVehicleInterface (Variables|GAVars|Data|GetOwningPlayerPawn))
    (:then
      (if (Vehicle|VehicleIsinAir (Vehicle|GetVehicleInterface (Variables|GAVars|Data|GetOwningPlayerPawn)))
        ;[1 1 1] Add the time spent in air to the total elapsed time in air (as calculated by the time interval)
        (bind _output_get (Variables|Default|SetElapsedTimeInAir (+ (Variables|Default|GetExecutionTimeInterval) (Variables|Default|GetElapsedTimeInAir))))
        (if (>= _output_get (Variables|Default|GetRequiredTime))
          _)
        (elif (> (Variables|Default|GetElapsedTimeInAir) 0)
          _))
      (Quest|SendComplexCustomStatEvent (Variables|GAVars|Data|GetOwningPlayerQuestManager) 0 (Variables|Default|GetT_Quest) "(GameplayTags=)" (Math|Float|Truncate (Variables|Default|GetElapsedTimeInAir)))
      (Variables|Default|SetElapsedTimeInAir 0.0))
    (:CastFailed
      (Ability|EndAbility))))

(event EventBP_OnQuestCompleted
  (params (QuestItem :type Fort%20Quest%20Item%20Object%20Reference))
  (Utilities|Time|ClearandInvalidateTimerbyHandle (Variables|Default|GetTimer%20Handle))
  (Ability|EndAbility))

(event Ability|EventActivateAbilityFromEvent
  (params (EventData :type Gameplay%20Event%20Data%20Structure%20%28by%20ref%29))
  (bind _timer_handle (Variables|Default|GetTimer%20Handle))
  (|CallParentFunction EventData)
  (switch Utilities|FlowControl|Switch|SwitchonGameplayTag (Utilities|Struct|BreakGameplayEventData EventData)
    (:Event.Vehicle.Entered
      ;[1 1 1] Check Vehicle Tags
      (if (GameplayTags|TagContainer|CompareTagAssetInterfacetoOtherTags (Variables|Default|GetT_Vehicle) (Vehicle|GetVehicleAbilitySystemComponent (Vehicle|GetVehicleInterface (Variables|GAVars|Data|GetOwningPlayerPawn))))
        ;[1 1 1] Start Checking Vehicle If Vehicle is in Air
        (bind _returnvalue (Utilities|Time|SetTimerbyEvent (delegate Custom|Check%20In%20Air) (Variables|Default|GetExecutionTimeInterval) true))
        (Variables|Default|SetTimer%20Handle _returnvalue)
        (else
          (Ability|EndAbility))))
    (:Event.Vehicle.Exit
      ;[1 1 1] On Vehicle Exit Clear Timer and End Ability
      (Variables|Default|SetElapsedTimeInAir 0.0)
      (if (Utilities|Time|IsValidTimerHandle _timer_handle)
        (Utilities|Time|ClearandInvalidateTimerbyHandle _timer_handle)
        (else
          _))
      (Ability|EndAbility))))
